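# Physics asset for the M_MED_Foray backpack cosmetic, giving its belt, handle and zipper bones collision shapes and joint limits.
# Each bone has one capsule body, with a large capsule covering the pack itself.
# Constraints link zipper and handle to their base bones and chain two belt segments from a belt base.
# Tight angular limits restrict motion: the handle only twists, while zipper and belt allow small swings and twists.
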PhysicsAsset /BRCosmetics/Accessories/FORT_Backpacks/M_MED_Foray/Meshes/M_MED_Foray_Pack_Physics
bodies {
 dyn_belt_01 [{shapeType=Capsule, length=3.8379, radius=1.0564, center=<1.6101 0.3958 0.049>, rotation=<rot 2.9567 -98.9315 67.2344>}]
 dyn_belt_02 [{shapeType=Capsule, length=3.0677, radius=0.9762, center=<1.8351 0.1393 -0.3618>, rotation=<rot 10.8988 -90 -75.2084>}]
 dyn_belt_base [{shapeType=Capsule, length=0, radius=2.4541, center=<2.1381 0.1069 -0.15>, rotation=<rot -5.8289 -90 -1.9755>}]
 dyn_handle [{shapeType=Capsule, length=5.3015, radius=1.5466, center=<0.005 -0.0504 3.7015>, rotation=<rot 0 91.5168 -90.5487>}]
 dyn_handle_base [{shapeType=Capsule, length=7.1966, radius=2.6065, center=<0.0178 -0.1064 0.9133>, rotation=<rot 0 -89.6854 -89.4839>}]
 dyn_zipper [{shapeType=Capsule, length=1.5818, radius=0.6786, center=<-0.0052 -0.0933 -1.1829>, rotation=<rot 3.4181 -90 0.19>}]
 dyn_zipper_base [{shapeType=Capsule, length=0.505, radius=0.505, rotation=<rot 0 -90 0>}]
 pack [{shapeType=Capsule, length=22.6163, radius=6.764, center=<7.7054 -5.3437 -9.4693>, rotation=<rot -2.0469 -90 1.0254>}]
}
constraints [
 {bone1Name=dyn_zipper, bone2Name=dyn_zipper_base, swing1LimitDegrees=1, swing2LimitDegrees=15, twistLimitDegrees=1}
 {bone1Name=dyn_handle, bone2Name=dyn_handle_base, swing1LimitDegrees=0, swing2LimitDegrees=0, twistLimitDegrees=30}
 {bone1Name=dyn_belt_01, bone2Name=dyn_belt_base, swing1LimitDegrees=10, swing2LimitDegrees=2, twistLimitDegrees=2}
 {bone1Name=dyn_belt_02, bone2Name=dyn_belt_01, swing1LimitDegrees=10, swing2LimitDegrees=2, twistLimitDegrees=2}
]
