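# Animation Blueprint for the M_MED_Foray backpack cosmetic, adding secondary physics motion and context-dependent sizing.
# From the reference pose, a Modify Bone node applies a scale variable to the pack bone in bone space.
# A RigidBody node then simulates the pack's physics asset relative to that bone, with damped component motion and overridden gravity.
# Each update shrinks the pack when bIsInFrontEnd is true and otherwise restores full size.
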AnimBlueprint /BRCosmetics/Accessories/FORT_Backpacks/M_MED_Foray/Meshes/M_MED_Foray_Pack_AnimBP
target_skeleton=&/M_MED_Foray_Pack_Skeleton
graphs [
 AnimGraph {nodes [
   n1 = "Transform (Modify) Bone - Bone: pack"(
    ComponentPose=Local To Component(LocalPose=Local Space Ref Pose())
    Translation="0,0,0"
    Rotation="0,0,0"
    Scale=Get scale() -> (scale:Vector)
   ) {
    scaleMode=BMM_Additive
    scaleSpace=BCS_BoneSpace
    alphaBoolBlend {blendOption=Linear}
    alphaScaleBiasClamp {bInterpResult=true, inRange {max=0, min=-40}, outRange {max=0, min=1}}
    boneToModify {boneName=pack}
    scale=<0 0 0>
   }
   n2 = RigidBody(ComponentPose=@n1) {
    bOverrideWorldGravity=true
    overridePhysicsAsset=&/M_MED_Foray_Pack_Physics
    simulationSpace=BaseBoneSpace
    alphaBoolBlend {blendOption=Linear}
    baseBoneRef {boneName=pack}
    componentAppliedLinearAccClamp=<800 800 800>
    componentLinearAccScale=<0.3 0.35 0.35>
    componentLinearVelScale=<0.35 0.35 0.35>
    overrideWorldGravity {z=-980}
   }
   Output Pose(Result=Component To Local(ComponentPose=@n2)) {layerGroup=None}
  ]}
 EventGraph {nodes [
   Event BlueprintUpdateAnimation()
   > n1 = Branch(Condition=Get bIsInFrontEnd())
   > Set scale(scale:Vector="0.85,0.85,0.85") -> (Output_Get:Vector)
   Set scale(execute=@n1.else, scale="1,1,1")
   TryGetPawnOwner()
  ]}
]
variables [scale {type=Vector}]
